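# Physics asset for the Wise Master face accessory cosmetic, covering the dynamic beard, eyebrows, sideburns and man bun.
# A four-link beard chain hangs from the jaw, while two five-link sideburn chains, both eyebrows and the bun attach to the head, all with angular limits.
# Capsules on the pelvis, upper spine, head and jaw form the body, joined by free-rotation constraints; another free constraint links the two sideburn roots.
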PhysicsAsset /BRCosmetics/Accessories/Hats/M_MED_Wise_Master_FaceAcc/Meshes/M_MED_Wise_Master_FaceAcc_Dyn
bodies {
 C_jaw [{shapeType=Capsule, length=1, radius=6}]
 dyn_beard_01 [{shapeType=Capsule, length=5.5, radius=4.25, shapeName=dyn_beard_01_capsule, center=<3.2652 0.0001 -1.315>, rotation=<rot 0 0 89.9999>}]
 dyn_beard_02 [{shapeType=Capsule, length=8, radius=4.25, center=<3.3897 0 0>, rotation=<rot 0 0 90>}]
 dyn_beard_03 [{shapeType=Sphere, radius=6, shapeName=dyn_beard_03_sphere, center=<1.9608 -0.0001 2>}]
 dyn_beard_04 [{shapeType=Sphere, radius=3.5, shapeName=dyn_beard_04_sphere, center=<2.8699 0 0>}]
 dyn_eyebrow_l [{shapeType=Capsule, length=4.0156, radius=1.005, rotation=<rot 90 0 0>}]
 dyn_eyebrow_r [{shapeType=Capsule, length=4.0156, radius=1.005, rotation=<rot 90 0 0>}]
 dyn_manbun_bk_01 [{shapeType=Capsule, length=2, radius=1, center=<0.1 0 0>, rotation=<rot 90 0 0>}]
 dyn_sideburn_l_01 [{shapeType=Capsule, length=6.5, radius=1.25, shapeName=dyn_sideburn_l_01_capsule, center=<4.8016 0 -0.0001>, rotation=<rot 90 0 0>}]
 dyn_sideburn_l_02 [{shapeType=Capsule, length=4.75, radius=1.5, shapeName=dyn_sideburn_l_02_capsule, center=<4.9548 0 -0.0004>, rotation=<rot 90 0 0>}]
 dyn_sideburn_l_03 [{shapeType=Capsule, length=8, radius=1.5, shapeName=dyn_sideburn_l_03_capsule, center=<4.9548 0 -0.0004>, rotation=<rot 90 0 0>}]
 dyn_sideburn_l_04 [{shapeType=Capsule, length=9, radius=1.5, shapeName=dyn_sideburn_l_04_capsule, center=<4.9548 0 -0.0004>, rotation=<rot 90 0 0>}]
 dyn_sideburn_l_05 [{shapeType=Capsule, length=9, radius=1.5, shapeName=dyn_sideburn_l_05_capsule, center=<4.9548 0 -0.0004>, rotation=<rot 90 0 0>}]
 dyn_sideburn_r_01 [{shapeType=Capsule, length=6.5, radius=1.25, shapeName=dyn_sideburn_r_01_capsule, center=<4.5697 0 0.0001>, rotation=<rot 90 0 0>}]
 dyn_sideburn_r_02 [{shapeType=Capsule, length=4.75, radius=1.5, shapeName=dyn_sideburn_r_02_capsule, center=<-4.356 0 -0.0001>, rotation=<rot 90 0 0>}]
 dyn_sideburn_r_03 [{shapeType=Capsule, length=8, radius=1.5, shapeName=dyn_sideburn_r_03_capsule, center=<-4.356 0 -0.0001>, rotation=<rot 90 0 0>}]
 dyn_sideburn_r_04 [{shapeType=Capsule, length=9, radius=1.5, shapeName=dyn_sideburn_r_04_capsule, center=<-4.356 0 -0.0001>, rotation=<rot 90 0 0>}]
 dyn_sideburn_r_05 [{shapeType=Capsule, length=9, radius=1.5, shapeName=dyn_sideburn_r_05_capsule, center=<-4.356 0 -0.0001>, rotation=<rot 90 0 0>}]
 head [{shapeType=Capsule, length=1, radius=1, rotation=<rot 90 0 0>}]
 pelvis [{shapeType=Capsule, length=7.5, radius=12.75}]
 spine_03 [{shapeType=Capsule, length=7.75, radius=18, center=<-4.1414 -4.6712 0>}]
 spine_04 [{shapeType=Capsule, length=18, radius=19.25, center=<-0.966 -5.5755 0>}]
 spine_05 [
  {shapeType=Capsule, length=27.5, radius=11.75, center=<-1.6651 -0.1688 0>}
  {shapeType=Capsule, length=20.5, radius=10.5, center=<3.7542 -1.0637 0>}
  {
   shapeType=Capsule
   length=20
   radius=10
   shapeName=spine_05_capsule
   center=<1.1217 0.3268 -23.5009>
   rotation=<rot -5.9694 -19.209 -30.6325>
  }
 ]
}
constraints [
 {bone1Name=dyn_manbun_bk_01, bone2Name=head, swing1LimitDegrees=30, swing2LimitDegrees=30, twistLimitDegrees=5}
 {
  bone1Name=C_jaw
  bone2Name=head
  swing1LimitDegrees=45
  swing1Motion=Free
  swing2LimitDegrees=45
  swing2Motion=Free
  twistLimitDegrees=45
  twistMotion=Free
 }
 {bone1Name=dyn_beard_01, bone2Name=C_jaw, swing1LimitDegrees=60, swing2LimitDegrees=60, twistLimitDegrees=15}
 {bone1Name=dyn_beard_02, bone2Name=dyn_beard_01, swing1LimitDegrees=60, swing2LimitDegrees=60, twistLimitDegrees=15}
 {bone1Name=dyn_beard_03, bone2Name=dyn_beard_02, swing1LimitDegrees=60, swing2LimitDegrees=60, twistLimitDegrees=5}
 {bone1Name=dyn_beard_04, bone2Name=dyn_beard_03, swing1LimitDegrees=60, swing2LimitDegrees=60, twistLimitDegrees=5}
 {
  bone1Name=head
  bone2Name=spine_05
  swing1LimitDegrees=45
  swing1Motion=Free
  swing2LimitDegrees=45
  swing2Motion=Free
  twistLimitDegrees=45
  twistMotion=Free
 }
 {
  bone1Name=spine_05
  bone2Name=spine_04
  swing1LimitDegrees=45
  swing1Motion=Free
  swing2LimitDegrees=45
  swing2Motion=Free
  twistLimitDegrees=45
  twistMotion=Free
 }
 {
  bone1Name=spine_04
  bone2Name=spine_03
  swing1LimitDegrees=45
  swing1Motion=Free
  swing2LimitDegrees=45
  swing2Motion=Free
  twistLimitDegrees=45
  twistMotion=Free
 }
 {
  bone1Name=spine_03
  bone2Name=pelvis
  swing1LimitDegrees=45
  swing1Motion=Free
  swing2LimitDegrees=45
  swing2Motion=Free
  twistLimitDegrees=45
  twistMotion=Free
 }
 {bone1Name=dyn_eyebrow_l, bone2Name=head, swing1LimitDegrees=45, swing2LimitDegrees=12, twistLimitDegrees=5}
 {bone1Name=dyn_eyebrow_r, bone2Name=head, swing1LimitDegrees=45, swing2LimitDegrees=12, twistLimitDegrees=5}
 {bone1Name=dyn_sideburn_l_01, bone2Name=head, swing1LimitDegrees=10, swing2LimitDegrees=10, twistLimitDegrees=5}
 {bone1Name=dyn_sideburn_l_02, bone2Name=dyn_sideburn_l_01, swing1LimitDegrees=30, swing2LimitDegrees=30, twistLimitDegrees=5}
 {bone1Name=dyn_sideburn_l_03, bone2Name=dyn_sideburn_l_02, swing1LimitDegrees=10, swing2LimitDegrees=10, twistLimitDegrees=5}
 {bone1Name=dyn_sideburn_l_04, bone2Name=dyn_sideburn_l_03, swing1LimitDegrees=10, swing2LimitDegrees=10, twistLimitDegrees=5}
 {bone1Name=dyn_sideburn_l_05, bone2Name=dyn_sideburn_l_04, swing1LimitDegrees=10, swing2LimitDegrees=10, twistLimitDegrees=5}
 {bone1Name=dyn_sideburn_r_01, bone2Name=head, swing1LimitDegrees=10, swing2LimitDegrees=10, twistLimitDegrees=5}
 {bone1Name=dyn_sideburn_r_02, bone2Name=dyn_sideburn_r_01, swing1LimitDegrees=30, swing2LimitDegrees=30, twistLimitDegrees=5}
 {bone1Name=dyn_sideburn_r_04, bone2Name=dyn_sideburn_r_03, swing1LimitDegrees=10, swing2LimitDegrees=10, twistLimitDegrees=5}
 {bone1Name=dyn_sideburn_r_03, bone2Name=dyn_sideburn_r_02, swing1LimitDegrees=10, swing2LimitDegrees=10, twistLimitDegrees=5}
 {bone1Name=dyn_sideburn_r_05, bone2Name=dyn_sideburn_r_04, swing1LimitDegrees=10, swing2LimitDegrees=10, twistLimitDegrees=5}
 {
  bone1Name=dyn_sideburn_l_01
  bone2Name=dyn_sideburn_r_01
  swing1LimitDegrees=45
  swing1Motion=Free
  swing2LimitDegrees=45
  swing2Motion=Free
  twistLimitDegrees=45
  twistMotion=Free
 }
]
