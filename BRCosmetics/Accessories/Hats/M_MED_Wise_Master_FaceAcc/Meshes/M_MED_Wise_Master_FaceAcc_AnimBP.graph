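# Animation Blueprint for the Wise Master face accessory, driving its beard and sideburn bones.
# A cached pose copies the attached parent, layering thigh and upper-arm branches from Body_Mesh.
# It feeds either a RigidBody simulation in spine_05 space or a constraint chain tying hair bones to jaw, head, spine and neighboring bones.
# The rigids_off flag selects the constraint fallback when rigid bodies are disabled or LOD exceeds 1.
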AnimBlueprint /BRCosmetics/Accessories/Hats/M_MED_Wise_Master_FaceAcc/Meshes/M_MED_Wise_Master_FaceAcc_AnimBP
target_skeleton=&/M_MED_Wise_Master_FaceAcc_Skeleton
graphs [
 AnimGraph {nodes [
   n1 = RigidBody(ComponentPose=Local To Component(LocalPose=Use cached pose 'startCache'())) {
    bClampLinearTranslationLimitToRefPose=true
    bOverrideWorldGravity=true
    overridePhysicsAsset=&/M_MED_Wise_Master_FaceAcc_Dyn
    simulationSpace=BaseBoneSpace
    alphaBoolBlend {blendOption=Linear}
    baseBoneRef {boneName=spine_05}
    componentAppliedLinearAccClamp=<3000 3000 3000>
    componentLinearAccScale=<0.3 0.2 0.2>
    componentLinearVelScale=<0.3 0.2 0.2>
    overrideWorldGravity {z=-980}
   }
   n2 = "Constraint - dyn_beard_03 (P)"(ComponentPose=Local To Component(LocalPose=Use cached pose 'startCache'()), Alpha=1, ConstraintWeights_0=1) {
    alphaBoolBlend {blendOption=Linear}
    boneToModify {boneName=dyn_beard_03}
    constraintSetup [{targetBone {boneName=spine_05}}]
    constraintWeights [1]
   }
   n3 = "Constraint - dyn_beard_02 (PT)"!(ComponentPose=@n2, Alpha=1, ConstraintWeights_0=1, ConstraintWeights_1:float=1) -> (Pose) {
    alphaBoolBlend {blendOption=Linear}
    boneToModify {boneName=dyn_beard_02}
    constraintSetup [{targetBone {boneName=dyn_beard_01}}, {transformType=Translation, targetBone {boneName=dyn_beard_03}}]
    constraintWeights [1,1]
   }
   n4 = "Constraint - dyn_beard_01 (PT)"!(ComponentPose=@n3, Alpha=1, ConstraintWeights_0=1, ConstraintWeights_1:float=1) -> (Pose) {
    alphaBoolBlend {blendOption=Linear}
    boneToModify {boneName=dyn_beard_01}
    constraintSetup [{targetBone {boneName=C_jaw}}, {transformType=Translation, targetBone {boneName=dyn_beard_02}}]
    constraintWeights [1,1]
   }
   n5 = "Constraint - dyn_beard_04 (R)"(ComponentPose=@n4, Alpha=1, ConstraintWeights_0=1) {
    alphaBoolBlend {blendOption=Linear}
    boneToModify {boneName=dyn_beard_04}
    constraintSetup [{transformType=Rotation, targetBone {boneName=spine_04}}]
    constraintWeights [1]
   }
   n6 = "Constraint - dyn_beard_05 (R)"(ComponentPose=@n5, Alpha=1, ConstraintWeights_0=1) {
    alphaBoolBlend {blendOption=Linear}
    boneToModify {boneName=dyn_beard_05}
    constraintSetup [{transformType=Rotation, targetBone {boneName=spine_03}}]
    constraintWeights [1]
   }
   n7 = "Constraint - dyn_sideburn_l_04 (P)"(ComponentPose=@n6, Alpha=1, ConstraintWeights_0=1) {
    alphaBoolBlend {blendOption=Linear}
    boneToModify {boneName=dyn_sideburn_l_04}
    constraintSetup [{targetBone {boneName=spine_05}}]
    constraintWeights [1]
   }
   n8 = "Constraint - dyn_sideburn_l_03 (PT)"!(ComponentPose=@n7, Alpha=1, ConstraintWeights_0=1, ConstraintWeights_1:float=1) -> (Pose) {
    alphaBoolBlend {blendOption=Linear}
    boneToModify {boneName=dyn_sideburn_l_03}
    constraintSetup [{targetBone {boneName=dyn_sideburn_l_02}}, {transformType=Translation, targetBone {boneName=dyn_sideburn_l_04}}]
    constraintWeights [1,1]
   }
   n9 = "Constraint - dyn_sideburn_l_02 (PT)"!(ComponentPose=@n8, Alpha=1, ConstraintWeights_0=1, ConstraintWeights_1:float=1) -> (Pose) {
    alphaBoolBlend {blendOption=Linear}
    boneToModify {boneName=dyn_sideburn_l_02}
    constraintSetup [{targetBone {boneName=dyn_sideburn_l_01}}, {transformType=Translation, targetBone {boneName=dyn_sideburn_l_03}}]
    constraintWeights [1,1]
   }
   n10 = "Constraint - dyn_sideburn_l_01 (PT)"!(ComponentPose=@n9, Alpha=1, ConstraintWeights_0=1, ConstraintWeights_1:float=1) -> (Pose) {
    alphaBoolBlend {blendOption=Linear}
    boneToModify {boneName=dyn_sideburn_l_01}
    constraintSetup [{targetBone {boneName=head}}, {transformType=Translation, targetBone {boneName=dyn_sideburn_l_02}}]
    constraintWeights [1,1]
   }
   n11 = "Constraint - dyn_sideburn_r_04 (P)"(ComponentPose=@n10, Alpha=1, ConstraintWeights_0=1) {
    alphaBoolBlend {blendOption=Linear}
    boneToModify {boneName=dyn_sideburn_r_04}
    constraintSetup [{targetBone {boneName=spine_05}}]
    constraintWeights [1]
   }
   n12 = "Constraint - dyn_sideburn_r_03 (PT)"!(ComponentPose=@n11, Alpha=1, ConstraintWeights_0=1, ConstraintWeights_1:float=1) -> (Pose) {
    alphaBoolBlend {blendOption=Linear}
    boneToModify {boneName=dyn_sideburn_r_03}
    constraintSetup [{targetBone {boneName=dyn_sideburn_r_02}}, {transformType=Translation, targetBone {boneName=dyn_sideburn_r_04}}]
    constraintWeights [1,1]
   }
   n13 = "Constraint - dyn_sideburn_r_02 (PT)"!(ComponentPose=@n12, Alpha=1, ConstraintWeights_0=1, ConstraintWeights_1:float=1) -> (Pose) {
    alphaBoolBlend {blendOption=Linear}
    boneToModify {boneName=dyn_sideburn_r_02}
    constraintSetup [{targetBone {boneName=dyn_sideburn_r_01}}, {transformType=Translation, targetBone {boneName=dyn_sideburn_r_03}}]
    constraintWeights [1,1]
   }
   n14 = "Constraint - dyn_sideburn_r_01 (PT)"!(ComponentPose=@n13, Alpha=1, ConstraintWeights_0=1, ConstraintWeights_1:float=1) -> (Pose) {
    alphaBoolBlend {blendOption=Linear}
    boneToModify {boneName=dyn_sideburn_r_01}
    constraintSetup [{targetBone {boneName=head}}, {transformType=Translation, targetBone {boneName=dyn_sideburn_r_02}}]
    constraintWeights [1,1]
   }
   Output Pose(Result=Blend Poses by bool(
     BlendPose_0=Component To Local(ComponentPose=@n14)
     BlendPose_1=Component To Local(ComponentPose=@n1)
     bActiveValue=Get rigids_off() -> (rigids_off:bool)
    ) {blendType=Linear}) {layerGroup=None}
   n15 = Layered blend per bone(
    BasePose=Copy Pose From Mesh() {bCopyCurves=true, bUseAttachedParent=true}
    BlendPoses_0=Copy Pose From Mesh(SourceMeshComponent=Get Body_Mesh())
   ) {layerSetup [{branchFilters [{boneName=thigh_l}, {boneName=thigh_r}, {boneName=upperarm_l}, {boneName=upperarm_r}]}]}
   Save cached pose 'startCache'(Pose=@n15) {cacheName=startCache}
  ]}
 EventGraph {nodes [
   Event BlueprintInitializeAnimation()
   > Set Body_Mesh(Body_Mesh=GetOwnerSkeletalMeshComponent())
   Event BlueprintUpdateAnimation()
   n1 = Get CurrentLOD()
   n2 = Get bRigidBodyEnabled()
   > Set rigids_off(rigids_off:bool=OR Boolean(A=NOT Boolean(A=@n2), B="integer > integer"(A=@n1, B=1) {is=title K2Node_CallFunction})) -> (Output_Get:bool)
   TryGetPawnOwner()
  ]}
]
variables [
 Body_Mesh {type=SkeletalMeshComponent}
 rigids_off {type=boolean}
]
